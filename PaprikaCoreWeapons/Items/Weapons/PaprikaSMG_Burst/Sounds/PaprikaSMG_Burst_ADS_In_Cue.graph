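# Sound cue for the Paprika burst SMG's aim-down-sights-in foley, played at reduced volume.
# A random node picks between two equally weighted wave variations and feeds a modulator that randomizes pitch from normal to slightly higher.
# It uses the weapon foley sound class and reload attenuation settings, with a concurrency override that prevents new instances once the limit is reached.
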SoundCue /PaprikaCoreWeapons/Items/Weapons/PaprikaSMG_Burst/Sounds/PaprikaSMG_Burst_ADS_In_Cue
root=Modulator_0
nodes [
 {id=Modulator_0, children [Random_0], properties {pitchMax=1.15, pitchMin=1}}
 {id=Random_0, children [WavePlayer_0,WavePlayer_2], properties {weights [1,1]}}
 {id=WavePlayer_0, properties {soundWaveAssetPtr=&/PaprikaSMG_Burst_ADS_In_01}}
 {id=WavePlayer_2, properties {soundWaveAssetPtr=&/PaprikaSMG_Burst_ADS_In_02}}
]
properties {
 attenuationSettings=&Fort_Weapon_Reload_Att
 bOverrideConcurrency=true
 maxDistance=2250
 soundClassObject=&Fort_WeaponFoley
 volumeMultiplier=0.301
 attenuationOverrides {falloffDistance=4100, attenuationShapeExtents {x=900}}
 concurrencyOverrides {resolutionRule=PreventNew}
}
