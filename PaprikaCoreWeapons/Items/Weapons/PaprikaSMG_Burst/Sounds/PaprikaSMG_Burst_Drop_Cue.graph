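# Sound cue for the drop sound of the PaprikaSMG_Burst weapon.
# A Random node picks one of three equally weighted drop waves, and a Modulator at the root varies its pitch and volume.
# Playback runs at reduced volume in the loot-collect sound class, with shared weapon-drop attenuation and the Weapon_Drop_StopOldest concurrency set.
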SoundCue /PaprikaCoreWeapons/Items/Weapons/PaprikaSMG_Burst/Sounds/PaprikaSMG_Burst_Drop_Cue
root=Modulator_1
nodes [
 {id=Modulator_1, children [Random_2]}
 {id=Random_2, children [WavePlayer_6,WavePlayer_7,WavePlayer_8], properties {weights [1,1,1]}}
 {id=WavePlayer_6, properties {soundWaveAssetPtr=&/PaprikaSMG_Burst_Drop_01}}
 {id=WavePlayer_7, properties {soundWaveAssetPtr=&/PaprikaSMG_Burst_Drop_02}}
 {id=WavePlayer_8, properties {soundWaveAssetPtr=&/PaprikaSMG_Burst_Drop_03}}
]
properties {
 attenuationSettings=&Fort_Weapon_Drop_Att
 maxDistance=3000
 soundClassObject=&Fort_Loot_Collect
 volumeMultiplier=0.6
 concurrencyOverrides {resolutionRule=PreventNew}
 concurrencySet [&Weapon_Drop_StopOldest]
}
