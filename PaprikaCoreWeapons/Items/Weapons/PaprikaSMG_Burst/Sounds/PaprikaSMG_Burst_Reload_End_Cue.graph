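# Sound cue for the end-of-reload foley of the Paprika burst SMG.
# A Random node picks one of three equally weighted wave variations and feeds it through a Modulator at the root.
# It uses the shared Fortnite weapon-reload attenuation and weapon foley sound class, with lowered volume, a concurrency override and the debug flag left on.
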SoundCue /PaprikaCoreWeapons/Items/Weapons/PaprikaSMG_Burst/Sounds/PaprikaSMG_Burst_Reload_End_Cue
root=Modulator_0
nodes [
 {id=Modulator_0, children [Random_0]}
 {id=Random_0, children [WavePlayer_0,WavePlayer_1,WavePlayer_5], properties {weights [1,1,1]}}
 {id=WavePlayer_0, properties {soundWaveAssetPtr=&/PaprikaSMG_Burst_Reload_End_01}}
 {id=WavePlayer_1, properties {soundWaveAssetPtr=&/PaprikaSMG_Burst_Reload_End_02}}
 {id=WavePlayer_5, properties {soundWaveAssetPtr=&/PaprikaSMG_Burst_Reload_End_03}}
]
properties {
 attenuationSettings=&Fort_Weapon_Reload_Att
 bDebug=true
 bOverrideConcurrency=true
 maxDistance=2250
 soundClassObject=&Fort_WeaponFoley
 volumeMultiplier=0.451
}
