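# Sound cue for the magazine-insert step of the Paprika burst SMG reload.
# A random node picks one of three insert recordings at equal weight to vary each playback.
# It routes to the weapon foley sound class with shared reload attenuation, at reduced volume.
# Concurrency override and the debug flag are switched on.
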SoundCue /PaprikaCoreWeapons/Items/Weapons/PaprikaSMG_Burst/Sounds/PaprikaSMG_Burst_Reload_Insert_Cue
root=Random_1
nodes [
 {id=Random_1, children [WavePlayer_0,WavePlayer_1,WavePlayer_2], properties {weights [1,1,1]}}
 {id=WavePlayer_0, properties {soundWaveAssetPtr=&/PaprikaSMG_Burst_Reload_Insert_01}}
 {id=WavePlayer_1, properties {soundWaveAssetPtr=&/PaprikaSMG_Burst_Reload_Insert_02}}
 {id=WavePlayer_2, properties {soundWaveAssetPtr=&/PaprikaSMG_Burst_Reload_Insert_03}}
]
properties {
 attenuationSettings=&Fort_Weapon_Reload_Att
 bDebug=true
 bOverrideConcurrency=true
 maxDistance=2250
 soundClassObject=&Fort_WeaponFoley
 volumeMultiplier=0.551
}
