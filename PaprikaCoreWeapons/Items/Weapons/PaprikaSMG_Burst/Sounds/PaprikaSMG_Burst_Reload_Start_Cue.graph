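# Sound cue for the start of the Paprika burst SMG reload.
# A Random node picks one of three reload-start wave variations with equal weights.
# It plays in the weapon foley sound class at reduced volume, with the shared weapon reload attenuation and overridden concurrency.
# The debug flag is left on.
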SoundCue /PaprikaCoreWeapons/Items/Weapons/PaprikaSMG_Burst/Sounds/PaprikaSMG_Burst_Reload_Start_Cue
root=Random_1
nodes [
 {id=Random_1, children [WavePlayer_3,WavePlayer_4,WavePlayer_5], properties {weights [1,1,1]}}
 {id=WavePlayer_3, properties {soundWaveAssetPtr=&/PaprikaSMG_Burst_Reload_Start_01}}
 {id=WavePlayer_4, properties {soundWaveAssetPtr=&/PaprikaSMG_Burst_Reload_Start_02}}
 {id=WavePlayer_5, properties {soundWaveAssetPtr=&/PaprikaSMG_Burst_Reload_Start_03}}
]
properties {
 attenuationSettings=&Fort_Weapon_Reload_Att
 bDebug=true
 bOverrideConcurrency=true
 maxDistance=2250
 soundClassObject=&Fort_WeaponFoley
 volumeMultiplier=0.451
}
